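# Material function for the MaskTempleExplosion FX that pulls vertices toward the camera while keeping their on-screen size.
# It adds the input WPO to the unoffset world position, takes the vector from the camera, and multiplies it by a negated OffsetPercent, tunable through the OffsetPercentScale parameter.
# This shrinks geometry about the camera position, so perspective cancels the change.
# The WPO output holds only this pull, without the input WPO.
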MaterialFunction /PistolCanary/Assets/FX/MaskTempleExplosion/MF_CameraOffsetButKeepSize
graph {nodes [
  n1 = Add(
   A=WorldPosition() {worldPositionShaderOffset=WPT_ExcludeAllShaderOffsets}.XYZ
   B=FunctionInput() {bUsePreviewValueAsDefault=true, inputName=WPO}
  )
  n2 = CameraPositionWS()
  n3 = FunctionInput() {bUsePreviewValueAsDefault=true, inputName=OffsetPercent, inputType=FunctionInput_Scalar, previewValue=<0.85 0 0 1>}
  n4 = Multiply(A=@n3, B=ScalarParameter() {defaultValue=1, parameterName=OffsetPercentScale})
  FunctionOutput(None=Multiply(A=Subtract(A=@n1, B=@n2), B=Multiply(A=@n4) {constB=-1})) {outputName=WPO}
 ]}
