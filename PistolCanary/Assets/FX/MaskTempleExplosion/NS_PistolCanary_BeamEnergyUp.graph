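# Niagara system for an upward pink-purple energy beam blast, part of the PistolCanary Mask Temple explosion effects.
# A ribbon beam spreads particles by execution index from below System.ImpactPos toward a rising end point, fading along its length.
# Layered flash sprites, meshes, a light, glow flares and shell bursts form the blast.
# GPU sparks swirl upward through curl, vortex and attraction forces, while smoke rises and rings outward from System.GroundPosition.
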NiagaraSystem /PistolCanary/Assets/FX/MaskTempleExplosion/NS_PistolCanary_BeamEnergyUp
dependencies {
 usedDataInterfaces [&NiagaraDataInterfaceCurve, &NiagaraDataInterfaceColorCurve, &NiagaraDataInterfaceVectorCurve]
 usedDynamicInputs [
  &MakeQuatFromAxisAngle
  &Divide_Float
  /Niagara/DynamicInputs/Math/Floor_Float
  &TransformPosition
 ]
 usedModules [
  &self:SystemSpawnScript.SystemScriptSource.SystemScriptGraph.NiagaraNodeAssignment_0.SetVariables_#1
  &self:SystemSpawnScript.SystemScriptSource.SystemScriptGraph.NiagaraNodeAssignment_1.SetVariables_#2
  &SystemState
 ]
}
emitters {
 AmbientSmoke {
  emitterSpawnScript [/SetVariables(Emitter.LightingCenter:Vector4f=&MakeFloat4FromVectorAndFloat(
     "Vector (RGB)"=&TransformVector(
      Vector=&AddVectorToPosition(Position=$Engine.Emitter.SimulationPosition, Vector=<0 0 100>)
      Source Space=::Simulation
      Destination Space=::World
     )
     "Float (Alpha)"=1
    )) {isSetParameters=true}]
  emitterUpdateScript [
   &EmitterState(Life Cycle Mode=::NewEnumerator1, Loop Behavior=::NewEnumerator1, Loop Duration=1, UseLoopDelay=true)
   &SpawnBurst_Instantaneous(Spawn Count=10, Use Spawn Probability=false, UseLoopCountLimit=false)
   &SpawnRate(SpawnRate=50, Use Spawn Probability=false)
  ]
  particleSpawnScript [
   &InitializeParticle(
    Lifetime=&RandomRangeFloat(Minimum=1, Maximum=3)
    Position Offset=$System.ImpactPos
    Sprite Rotation Mode=::NewEnumerator2
    Sprite Size Mode=::NewEnumerator3
    Uniform Sprite Size=750
    UsePositionOffset=true
    Write Material Random=true
   ) {module=InitializeParticle001}
   &ShapeLocation(
    Cylinder Height=20
    Cylinder Radius=&RandomRangeFloat(Minimum=1500, Maximum=2500)
    MaskSpawnGroup=false
    Offset Mode=::NewEnumerator0
    Override Seed=false
    Shape Primitive=::NewEnumerator1
    Surface Only=false
    UseOwnerScale=true
   )
   &CameraOffset()
  ]
  particleUpdateScript [
   &ParticleState()
   /SetVariables(Particles.Position=&Add_Vector(
     A=&Lerp_Vector(
      A=Linked() {linkedVariable {name=Particles.Initial.Position, type=NiagaraPosition}}
      B=<0 0 0>
      Alpha=&FloatFromCurve(FloatCurve=DataInterface() {propertyValues {Curve {keys [
           {arriveTangent=2.3622, interpMode=RCIM_Cubic, leaveTangent=2.3622, tangentMode=RCTM_User, time=0, value=0}
           {arriveTangent=0.3265, interpMode=RCIM_Cubic, leaveTangent=0.3265, tangentMode=RCTM_User, time=1, value=1}
          ]}}})
     )
     B=&Multiply_VectorByFloat(
      Vector=&RandomRangeVector(Minimum=<0 0 250>, Maximum=<0 0 550>)
      Float=&FloatFromCurve(FloatCurve=DataInterface() {propertyValues {Curve {keys [
           {interpMode=RCIM_Cubic, tangentMode=RCTM_Break, time=0, value=0}
           {arriveTangent=2.7423, interpMode=RCIM_Cubic, tangentMode=RCTM_Break, time=1, value=1}
          ]}}})
     )
    )) {isSetParameters=true}
   &Drag(Use Linear Drag=true, Use Rotational Drag=false)
   &SolveForcesAndVelocity(Clamp Velocity=false, Limit Acceleration=false)
   &ScaleSpriteSize(Scale Factor=&Vector2DFromFloat(Value=&FloatFromCurve(FloatCurve=DataInterface() {propertyValues {Curve {keys [{interpMode=RCIM_Cubic, leaveTangent=1.1992, tangentMode=RCTM_Break, time=0, value=0.25}, <at 0.8 1>]}}})))
   &ScaleColor(
    Scale Alpha=&FloatFromCurve(
     FloatCurve=DataInterface() {propertyValues {Curve {keys [
         {interpMode=RCIM_Cubic, leaveTangent=7.3008, tangentMode=RCTM_Break, time=0, value=0}
         {interpMode=RCIM_Cubic, tangentMode=RCTM_Break, time=0.2, value=1}
         {interpMode=RCIM_Cubic, leaveTangent=-0.0057, tangentMode=RCTM_Break, time=0.5, value=1}
         <at 1 0 RCIM_Cubic>
        ]}}}
     Scale Curve=&RandomRangeFloat(Minimum=0.3, Maximum=0.1)
    )
    Scale RGB=&MakeVectorFromLinearColorRGB(Color=&ColorFromCurve(ColorCurve=DataInterface() {propertyValues {
        AlphaCurve {keys [<at 0 1 RCIM_Cubic>,<at 1 0 RCIM_Cubic>]}
        BlueCurve {keys [<at 0 0.03 RCIM_Cubic>]}
        GreenCurve {keys [<at 0 0.03 RCIM_Cubic>]}
        RedCurve {keys [<at 0 0.03 RCIM_Cubic>]}
       }}))
    ScaleA=true
    ScaleRGB=true
   )
   /SetVariables(Particles.SubImageIndex=&FloatFromCurve(FloatCurve=DataInterface() {propertyValues {Curve {keys [<at 0 0>,<at 1 1>]}}}, Scale Curve=16)) {isSetParameters=true}
   &DynamicMaterialParameters(
    Index 0 Param 4=&FloatFromCurve(FloatCurve=DataInterface() {propertyValues {Curve {keys [<at 0 1>,<at 1 0>]}}}, Scale Curve=2)
    Module.Index0Param0Write=true
    Module.Index0Param1Write=true
    Module.Index0Param2Write=true
    Module.Index0Param3Write=true
   )
  ]
  rendererClasses [&NiagaraSpriteRendererProperties]
  renderers [{rendererClass=&NiagaraSpriteRendererProperties}]
 }
 DirectionalBurst {
  emitterUpdateScript [
   &EmitterState(Life Cycle Mode=::NewEnumerator1, Loop Behavior=::NewEnumerator1, Loop Duration=1.5, UseLoopDelay=false)
   &SpawnBurst_Instantaneous(Use Spawn Probability=false, UseLoopCountLimit=false)
   &SpawnRate(SpawnRate=6, Use Spawn Probability=false) {enabled=false}
  ]
  particleSpawnScript [
   &InitializeParticle(
    Color=<rgba 1.2 0.7583 0.9569 0.5>
    Mass Max=2
    Mass Mode=::NewEnumerator3
    Mesh Scale Mode=::NewEnumerator3
    Mesh Uniform Scale=3.2
    Position Offset=<0 0 -125>
    UsePositionOffset=true
    Write Material Random=true
   )
   &DynamicMaterialParameters(
    Index 0 Param 1=1
    Index 0 Param 2=0.4
    Module.Index0Float0Write=true
    Module.Index0Float1Write=true
    Module.Index0Float2Write=true
    Module.Index0Float3Write=true
   )
  ]
  particleUpdateScript [
   &ParticleState()
   &ScaleColor(
    Scale Alpha=&FloatFromCurve(FloatCurve=DataInterface() {propertyValues {Curve {keys [
         {arriveTangent=-0.9969, interpMode=RCIM_Cubic, time=0, value=0}
         {arriveTangent=0.002, interpMode=RCIM_Cubic, leaveTangent=0.002, time=0.1, value=0.95}
         {interpMode=RCIM_Cubic, leaveTangent=0.002, time=1.0011, value=0.002}
        ]}}})
    Scale RGB=&MakeVectorFromLinearColorRGB(Color=<rgba 1 1 1 1>)
    ScaleA=true
    ScaleRGB=true
   )
   &ScaleMeshSize(Scale Factor=&VectorFromCurve(
     VectorCurve=DataInterface() {propertyValues {
       XCurve {keys [
         {arriveTangent=3.3253, interpMode=RCIM_Cubic, leaveTangent=3.3253, tangentMode=RCTM_User, time=0, value=0.2}
         {arriveTangent=0.1373, interpMode=RCIM_Cubic, leaveTangent=0.1373, tangentMode=RCTM_User, time=0.5, value=1.1}
         <at 1 1.25 RCIM_Cubic>
        ]}
       YCurve {keys [
         {arriveTangent=2.3637, interpMode=RCIM_Cubic, leaveTangent=2.3637, tangentMode=RCTM_User, time=0, value=0.3}
         {interpMode=RCIM_Cubic, tangentMode=RCTM_User, time=1, value=1}
        ]}
       ZCurve {keys [
         {arriveTangent=2.3637, interpMode=RCIM_Cubic, leaveTangent=2.3637, tangentMode=RCTM_User, time=0, value=0.3}
         {interpMode=RCIM_Cubic, tangentMode=RCTM_User, time=1, value=1}
        ]}
      }}
     Scale Curve=<1 10 10>
    ))
  ]
  rendererClasses [&NiagaraMeshRendererProperties]
  renderers [{rendererClass=&NiagaraMeshRendererProperties, rendererIndex=0}]
 }
 GPU_Sparks {
  simTarget=GPUComputeSim
  emitterUpdateScript [
   &EmitterState(Life Cycle Mode=::NewEnumerator1, Loop Behavior=::NewEnumerator1, Loop Duration=1.7, UseLoopDelay=false)
   &SpawnRate(
    SpawnRate=&Multiply_Float(
     A=150
     B=&FloatFromCurve(FloatCurve=DataInterface() {propertyValues {Curve {keys [<at 0 1>,<at 1 0>]}}}, CurveIndex=$Emitter.NormalizedLoopAge)
    )
    Use Spawn Probability=false
   )
  ]
  particleSpawnScript [
   &InitializeParticle(
    Color=<rgba 1 0.4 0.7 1>
    Lifetime=&RandomRangeFloat(Minimum=1.5, Maximum=2)
    Position Offset=$System.ImpactPos
    Sprite Size=&Multiply_Vector2D_ByFloat(Float=&RandomRangeFloat(Minimum=20, Maximum=30), Vector2D=<1 3>)
    Sprite Size Mode=::NewEnumerator4
    UsePositionOffset=true
    Write Material Random=true
   )
   /SetVariables(Particles.MyRandom:NiagaraFloat=&RandomRangeFloat(Minimum=0, Maximum=1)) {isSetParameters=true}
   &ShapeLocation(Handle Radius=1, Large Radius=500, MaskSpawnGroup=false, Override Seed=false, Shape Primitive=::NewEnumerator3, UseOwnerScale=true)
  ]
  particleUpdateScript [
   &ParticleState()
   &ScaleSpriteSize(
    Scale Sprite Size Mode=::NewEnumerator0
    Uniform Scale Factor=&RandomRangeFloat(Minimum=0.5, Maximum=1.2)
   )
   &CurlNoiseForce(Bias Noise=false, Noise Frequency=2, Noise Strength=2000, Pan Noise=true, Pan Noise Field=<0 0 0.1>)
   &PointAttractionForce(
    Attraction Radius=1500000
    Attraction Strength=80
    Attractor Position Offset:Vector3f=<0 0 2500>
    Kill Within Radius=false
    Module.Use Falloff=true
    UseAttractorOffset=true
   )
   &VortexForce(Origin Pull Amount=500, UseOriginOffset=false, Vortex Force Amount=500)
   &Drag(Drag=&RandomRangeFloat(Minimum=1.5, Maximum=2), Use Linear Drag=true, Use Rotational Drag=true)
   &GravityForce(Gravity=&MakeVector(
     X=0
     Y=0
     Z=&FloatFromCurve(
      FloatCurve=DataInterface() {propertyValues {Curve {keys [{interpMode=RCIM_Cubic, tangentMode=RCTM_Break, time=0.4, value=0}, <at 1 1>]}}}
      Scale Curve=23267.6934
     )
    ))
   &SolveForcesAndVelocity(Clamp Velocity=false, Limit Acceleration=false)
   &ScaleColor(
    Scale Alpha=&FloatFromCurve(FloatCurve=DataInterface() {propertyValues {Curve {keys [
         {arriveTangent=-0.0049, interpMode=RCIM_Cubic, leaveTangent=-0.0049, tangentMode=RCTM_User, time=0, value=0}
         {arriveTangent=-0.0749, interpMode=RCIM_Cubic, leaveTangent=-0.0749, tangentMode=RCTM_User, time=0.3, value=0.9847}
         {arriveTangent=-0.0405, interpMode=RCIM_Cubic, leaveTangent=-0.0405, tangentMode=RCTM_User, time=1, value=0}
        ]}}})
    ScaleA=true
    ScaleRGB=true
   )
   &FN_Sprite_Attributes(
    Color=&MultiplyLinearColor(Linear Color=$Particles.Color, Scale Factor=<1 1 1 0.05>)
    Create Color Attribute=true
    Create SpriteSize Attribute=true
    SpriteSize=&Multiply_Vector2D(A=$Particles.SpriteSize, B=<10 5>)
    Write to Module Namespace=true
   )
  ]
  rendererClasses [&NiagaraSpriteRendererProperties]
  renderers [
   {rendererClass=&NiagaraSpriteRendererProperties}
   {rendererClass=&NiagaraSpriteRendererProperties, rendererIndex=1}
  ]
 }
 GlowFlare {
  emitterUpdateScript [
   &EmitterState()
   &SpawnBurst_Instantaneous(Use Spawn Probability=false, UseLoopCountLimit=true)
  ]
  particleSpawnScript [
   &InitializeParticle(
    Color=<rgba 0.012 0 0.02 0.4696>
    Lifetime=3.5
    Position Offset=<0 250 250>
    Sprite Size Mode=::NewEnumerator3
    Uniform Sprite Size=4500
    UsePositionOffset=true
    Write Material Random=true
   )
   /SetVariables(Particles.FlareCameraOffset1=22.8579, Particles.FlareCameraOffset2=20) {isSetParameters=true, module=SetVariables_#3}
   &ShapeLocation(MaskSpawnGroup=false, Override Seed=false, Sphere Radius=25, UseOwnerScale=true) {enabled=false}
  ]
  particleUpdateScript [
   &ParticleState()
   /SetVariables(
    Particles.FlareColor1:LinearColor=&MultiplyLinearColor(
     Linear Color=<rgba 0.9844 0.0122 0.1234 0.3478>
     Scale Factor=&MakeFloat4FromVectorAndFloat(
      "Vector (RGB)"=&VectorFromFloat(Value=&RandomRangeFloat(Minimum=0.9, Maximum=1, Evaluation Type=::NewEnumerator1))
      "Float (Alpha)"=1
     )
    )
    Particles.FlareColor2:LinearColor=&MultiplyLinearColor(
     Linear Color=<rgba 0.6042 0.053 0.1614 1>
     Scale Factor=&MakeFloat4FromVectorAndFloat(
      "Vector (RGB)"=&VectorFromFloat(Value=&RandomRangeFloat(Minimum=0.5, Maximum=2, Evaluation Type=::NewEnumerator1))
      "Float (Alpha)"=1
     )
    )
    Particles.FlareSize1:Vector2f=&Vector2DFromFloat(Value=&Multiply_Float(A=1500, B=&RandomRangeFloat(Minimum=0.9, Maximum=1, Evaluation Type=::NewEnumerator1)))
    Particles.FlareSize2:Vector2f=&Vector2DFromFloat(Value=&Multiply_Float(A=1000, B=&RandomRangeFloat(Minimum=0.5, Maximum=1, Evaluation Type=::NewEnumerator1)))
    Particles.GlowColor=&MultiplyLinearColor(
     Linear Color=<rgba 24.0418 0.015 40 0.2>
     Scale Factor=&MakeVector4FromVectorAndFloat(
      "Vector (RGB)"=<1 1 1>
      "Float (Alpha)"=&FloatFromCurve(FloatCurve=DataInterface() {propertyValues {Curve {
          defaultValue=1
          keys [
           <at 0 0 RCIM_Cubic>
           {arriveTangent=0.0746, interpMode=RCIM_Cubic, leaveTangent=0.0746, tangentMode=RCTM_User, time=0.1, value=1}
           {arriveTangent=-0.0405, interpMode=RCIM_Cubic, leaveTangent=-0.0405, tangentMode=RCTM_User, time=0.7009, value=1}
           {arriveTangent=-8.0089, interpMode=RCIM_Cubic, leaveTangent=-8.0089, tangentMode=RCTM_User, time=1, value=0}
          ]
         }}})
     )
    )
    Particles.GlowSize=&Vector2DFromFloat(Value=2400)
   ) {isSetParameters=true, module=SetVariables_#4}
   &ScaleColor(ScaleA=true, ScaleRGB=true)
   &FN_Light_Attributes(
    Color=<rgba 0.8637 2.4719 3 1>
    Create Color Attribute=true
    Create Exponent Attribute=true
    Create Radius Attribute=true
    Exponent=4
    Radius=250
    Scale Color=8
   )
   &DynamicMaterialParameters(
    Index 0 Param 1=1
    Module.Index0Float0Write=true
    Module.Index0Float1Write=true
    Module.Index0Float2Write=true
    Module.Index0Float3Write=true
   )
  ]
  rendererClasses [&NiagaraSpriteRendererProperties]
  renderers [{rendererClass=&NiagaraSpriteRendererProperties}]
 }
 GlowFlare001 {
  emitterUpdateScript [
   &EmitterState()
   &SpawnBurst_Instantaneous(Use Spawn Probability=false, UseLoopCountLimit=true)
  ]
  particleSpawnScript [
   &InitializeParticle(Color=<rgba 0.012 0 0.02 0.4696>, Lifetime=1.25, Position Offset=<0 250 250>, UsePositionOffset=true, Write Material Random=true)
   /SetVariables(Particles.FlareCameraOffset1=22.8579, Particles.FlareCameraOffset2=20) {isSetParameters=true, module=SetVariables_#3}
   &ShapeLocation(MaskSpawnGroup=false, Override Seed=false, Sphere Radius=25, UseOwnerScale=true) {enabled=false}
   &CameraOffset(Camera Offset Amount=4250)
  ]
  particleUpdateScript [
   &ParticleState()
   /SetVariables(
    Particles.GlowColor=&MultiplyLinearColor(
     Linear Color=<rgba 0.601 0.0004 1 0.125>
     Scale Factor=&MakeVector4FromVectorAndFloat(
      "Vector (RGB)"=<1 1 1>
      "Float (Alpha)"=&FloatFromCurve(FloatCurve=DataInterface() {propertyValues {Curve {
          defaultValue=1
          keys [
           <at 0 0 RCIM_Cubic>
           {arriveTangent=0.0746, interpMode=RCIM_Cubic, leaveTangent=0.0746, tangentMode=RCTM_User, time=0.1, value=0.95}
           {arriveTangent=-0.0405, interpMode=RCIM_Cubic, leaveTangent=-0.0405, tangentMode=RCTM_User, time=0.4, value=0.95}
           {arriveTangent=-3.5583, interpMode=RCIM_Cubic, leaveTangent=-3.5583, tangentMode=RCTM_User, time=1, value=0}
          ]
         }}})
     )
    )
    Particles.GlowSize=&Vector2DFromFloat(Value=17500)
   ) {isSetParameters=true, module=SetVariables_#4}
   &DynamicMaterialParameters(
    Index 0 Param 1=1
    Module.Index0Float0Write=true
    Module.Index0Float1Write=true
    Module.Index0Float2Write=true
    Module.Index0Float3Write=true
   )
  ]
  rendererClasses [&NiagaraSpriteRendererProperties]
  renderers [{rendererClass=&NiagaraSpriteRendererProperties}]
 }
 NE_LaserRibbon {
  emitterUpdateScript [
   &EmitterState()
   &SpawnBurst_Instantaneous(Spawn Count=10, Use Spawn Probability=false, UseLoopCountLimit=false)
  ]
  particleSpawnScript [
   &InitializeParticle(
    Color Mode=::NewEnumerator0
    Lifetime=2
    Ribbon Width=1500
    Ribbon Width Mode=::NewEnumerator1
    UsePositionOffset=false
    Write Material Random=true
   )
   &CameraOffset(Camera Offset Amount=150)
  ]
  particleUpdateScript [
   &ParticleState()
   /SetVariables(Particles.Position=&Lerp_Vector(
     A=&SubtractVectorFromPosition(Position=Linked() {linkedVariable {name=System.ImpactPos, type=Vector3f}}, Vector=<0 0 500>)
     B=&Lerp_Vector(
      A=$System.ImpactPos
      B=<0 0 20000>
      Alpha=&FloatFromCurve(FloatCurve=DataInterface() {propertyValues {Curve {keys [<at 0 0>,<at 1 1>]}}})
     )
     Alpha=&ReturnNormalizedExecIndex()
    )) {isSetParameters=true}
   &ScaleColor(
    Scale Alpha=&Multiply_Float(
     A=&FloatFromCurve(FloatCurve=DataInterface() {propertyValues {Curve {keys [
          {interpMode=RCIM_Cubic, tangentMode=RCTM_User, time=0.5, value=1}
          {arriveTangent=-2.6146, interpMode=RCIM_Cubic, leaveTangent=-2.6146, tangentMode=RCTM_User, time=1, value=0}
         ]}}})
     B=&FloatFromCurve(
      FloatCurve=DataInterface() {propertyValues {Curve {keys [<at 0.5 1>,<at 1 0>]}}}
      CurveIndex=&ReturnNormalizedExecIndex()
     )
    )
    Scale RGB=&VectorFromFloat(Value=&FloatFromCurve(FloatCurve=DataInterface() {propertyValues {Curve {
         defaultValue=1
         keys [{interpMode=RCIM_Cubic, tangentMode=RCTM_User, time=0, value=1}, {interpMode=RCIM_Cubic, tangentMode=RCTM_User, time=1, value=0}]
        }}}))
    ScaleA=true
    ScaleRGB=true
   )
   &ScaleRibbonWidth(Ribbon Width Scale=&FloatFromCurve(FloatCurve=DataInterface() {propertyValues {Curve {keys [
         <at 0 0 RCIM_Cubic>
         {arriveTangent=2, interpMode=RCIM_Cubic, leaveTangent=2, time=0.2, value=1}
         {interpMode=RCIM_Cubic, tangentMode=RCTM_User, time=0.5, value=1}
         {arriveTangent=-1.961, interpMode=RCIM_Cubic, leaveTangent=-1.961, tangentMode=RCTM_User, time=1, value=0.5}
        ]}}}))
   /SetVariables(
    Particles.GlowColor=&MultiplyLinearColor(Linear Color=$Particles.Color, Scale Factor=<1 1 1 1>)
    Particles.GlowWidth:NiagaraFloat=&Multiply_Float(A=$Particles.RibbonWidth, B=1)
   ) {isSetParameters=true}
   /SetVariables(
    Particles.CoreColor:LinearColor=&MultiplyLinearColor(
     Linear Color=$Particles.Color
     Scale Factor=&MakeVector4_FromLinearColor(LinearColor=&MakeLinearColorFromVectorAndFloat(
       "Vector (RGB)"=&VectorFromFloat(Value=&FloatFromCurve(FloatCurve=DataInterface() {propertyValues {Curve {keys [{arriveTangent=-279.7298, interpMode=RCIM_Cubic, leaveTangent=-279.7298, tangentMode=RCTM_User, time=0, value=100}, <at 1 10>]}}}))
       "Float (Alpha)"=1
      ))
    )
    Particles.CoreWidth:NiagaraFloat=&Multiply_Float(A=$Particles.RibbonWidth, B=0.25)
   ) {isSetParameters=true}
  ]
  rendererClasses [&NiagaraRibbonRendererProperties]
  renderers [
   {rendererClass=&NiagaraRibbonRendererProperties}
   {rendererClass=&NiagaraRibbonRendererProperties, rendererIndex=1}
   {rendererClass=&NiagaraRibbonRendererProperties, rendererIndex=2}
  ]
 }
 NE_MuzzleFlash {
  emitterUpdateScript [
   &EmitterState()
   &SpawnBurst_Instantaneous(Use Spawn Probability=false, UseLoopCountLimit=false)
  ]
  particleSpawnScript [
   &InitializeParticle(
    Color Mode=::NewEnumerator0
    Lifetime=2
    Mesh Scale Mode=::NewEnumerator3
    Position Offset=&Add_Vector(A=$System.ImpactPos, B=<0 0 250>)
    Sprite Size Mode=::NewEnumerator3
    Uniform Sprite Size=5000
    UsePositionOffset=true
    Write Material Random=true
   )
   &InitialMeshOrientation(Orientation Axis=<0 0 1>, Orientation Vector=<0 0 1>, Rotation=<0 0.25 0>, Use Orientation Vector=true, Use Rotation Vector=true)
   &CameraOffset(Camera Offset Amount=150)
   &SpriteFacingAndAlignment(Sprite Alignment=<0 0 1>, WriteAlignment=true, WriteFacing=false)
  ]
  particleUpdateScript [
   &ParticleState()
   &ScaleSpriteSize(Uniform Curve Sprite Scale=DataInterface() {propertyValues {Curve {keys [
        {arriveTangent=2.1756, interpMode=RCIM_Cubic, leaveTangent=2.1756, tangentMode=RCTM_User, time=0, value=0}
        {interpMode=RCIM_Cubic, tangentMode=RCTM_User, time=1, value=1}
       ]}}})
   &Color(Color=&ColorFromCurve(ColorCurve=DataInterface() {propertyValues {
       AlphaCurve {keys [<at 0.5878 1 RCIM_Cubic>,<at 0.9954 0>]}
       BlueCurve {keys [<at 0 1.2483 RCIM_Cubic>,<at 1 0.8485 RCIM_Cubic>]}
       GreenCurve {keys [<at 0 0 RCIM_Cubic>,<at 1 0.1778 RCIM_Cubic>]}
       RedCurve {keys [<at 0 5 RCIM_Cubic>,<at 1 1 RCIM_Cubic>]}
      }}))
   /SetVariables(Particles.LongFlareSize:Vector2f=&Multiply_Vector2D(A=$Particles.SpriteSize, B=<0.5 10>)) {isSetParameters=true}
   &FN_Sprite_Attributes(
    Color=&MultiplyLinearColor(Linear Color=$Particles.Color, Scale Factor=&MakeVector4_FromLinearColor(LinearColor=<rgba 1 1 1 0.5>))
    Create Color Attribute=true
    Create SpriteSize Attribute=true
    SpriteSize=&Multiply_Vector2D_ByFloat(Float=1, Vector2D=$Particles.SpriteSize)
    Write to Module Namespace=true
   )
   &FN_Sprite_Attributes(
    Color=&MultiplyLinearColor(
     Linear Color=$Particles.Color
     Scale Factor=&MakeVector4_FromLinearColor(LinearColor=<rgba 0.02 0.02 0.02 1>)
    )
    Create Color Attribute=true
    Create SpriteSize Attribute=true
    SpriteSize=&Multiply_Vector2D_ByFloat(Float=1.5, Vector2D=$Particles.SpriteSize)
    Write to Module Namespace=true
   ) {module=FN_Sprite_Attributes001}
   &FN_Light_Attributes(
    Color=$Particles.Color
    Create Color Attribute=true
    Create Exponent Attribute=true
    Create Radius Attribute=true
    Enable Particle Light Enable=true
    Exponent=2
    Light Enabled=true
    Radius=1500
    Scale Color=&FloatFromCurve(
     FloatCurve=DataInterface() {propertyValues {Curve {keys [
         {arriveTangent=4.8174, interpMode=RCIM_Cubic, leaveTangent=4.8174, tangentMode=RCTM_User, time=0, value=0}
         {interpMode=RCIM_Cubic, tangentMode=RCTM_User, time=0.15, value=1}
         {interpMode=RCIM_Cubic, tangentMode=RCTM_User, time=1, value=0}
        ]}}}
     Scale Curve=10
    )
   )
   &DynamicMaterialParameters(
    Index 3 Param 1=1
    Index 3 Param 4=&Multiply_Float(A=$Engine.System.Age, B=-2)
    Module.Index3Float0Write=true
    Module.Index3Float1Write=true
    Module.Index3Float2Write=true
    Module.Index3Float3Write=true
    Write Parameter Index 0=::NewEnumerator0
    Write Parameter Index 3=::NewEnumerator1
   )
   &FN_Mesh_Attributes(
    Color=&MultiplyLinearColor(
     Linear Color=&ColorFromCurve(ColorCurve=DataInterface() {propertyValues {
        AlphaCurve {keys [<at 0 1 RCIM_Cubic>]}
        BlueCurve {keys [<at 0 0.9942 RCIM_Cubic>,<at 0.6269 0.6666 RCIM_Cubic>]}
        GreenCurve {keys [<at 0 0.1305 RCIM_Cubic>,<at 0.6269 0.2094 RCIM_Cubic>]}
        RedCurve {keys [<at 0 1 RCIM_Cubic>,<at 0.6269 1 RCIM_Cubic>]}
       }})
     Scale Factor=&MakeFloat4FromVectorAndFloat(
      "Vector (RGB)"=&VectorFromFloat(Value=&FloatFromCurve(
        FloatCurve=DataInterface() {propertyValues {Curve {keys [
            <at 0 0 RCIM_Cubic>
            {interpMode=RCIM_Cubic, tangentMode=RCTM_Break, time=0.1, value=1}
            {arriveTangent=-2.2558, interpMode=RCIM_Cubic, tangentMode=RCTM_Break, time=1, value=0}
           ]}}}
        Scale Curve=2
       ))
      "Float (Alpha)"=&FloatFromCurve(FloatCurve=DataInterface() {propertyValues {Curve {keys [<at 0.4 1 RCIM_Cubic>,<at 1 0>]}}})
     )
    )
    Create Color Attribute=true
    Create Scale Attribute=true
    Scale=&VectorFromFloat(Value=&FloatFromCurve(
      FloatCurve=DataInterface() {propertyValues {Curve {keys [
          {arriveTangent=4.7021, interpMode=RCIM_Cubic, leaveTangent=4.7021, tangentMode=RCTM_User, time=0, value=0}
          {arriveTangent=2.0471, interpMode=RCIM_Cubic, leaveTangent=2.0471, tangentMode=RCTM_User, time=0.3, value=1}
          {interpMode=RCIM_Cubic, tangentMode=RCTM_User, time=1, value=0.2}
         ]}}}
      Scale Curve=20
     ))
    Write to Module Namespace=true
   ) {module=FN_Mesh_Attributes001}
   /SetVariables(Particles.EnergyMeshDynMat:Vector4f=&MakeVector4(
     X=&FloatFromCurve(FloatCurve=DataInterface() {propertyValues {Curve {keys [{arriveTangent=-2.1015, interpMode=RCIM_Cubic, leaveTangent=-2.1015, tangentMode=RCTM_User, time=0, value=0.5}, <at 1 -0.15>]}}})
     Y=&FloatFromCurve(FloatCurve=DataInterface() {propertyValues {Curve {keys [{arriveTangent=-0.3464, interpMode=RCIM_Cubic, leaveTangent=-0.3464, tangentMode=RCTM_User, time=0, value=1}, <at 1 0.9>]}}})
     Z=&FloatFromCurve(FloatCurve=DataInterface() {propertyValues {Curve {keys [
          {interpMode=RCIM_Cubic, tangentMode=RCTM_Break, time=0, value=0}
          {arriveTangent=0.4478, interpMode=RCIM_Cubic, leaveTangent=0.4478, tangentMode=RCTM_User, time=0.15, value=1}
          {interpMode=RCIM_Cubic, tangentMode=RCTM_Break, time=1, value=1.2}
         ]}}})
     W=0
    )) {isSetParameters=true}
   &FN_Mesh_Attributes(
    Color=&MultiplyLinearColor(
     Linear Color=&ColorFromCurve(ColorCurve=DataInterface() {propertyValues {
        AlphaCurve {keys [<at 0 1 RCIM_Cubic>]}
        BlueCurve {keys [
          <at 0 0.9942>
          {arriveTangent=-0.5226, interpMode=RCIM_Cubic, leaveTangent=-0.5226, time=0.0999, value=9.9422}
          <at 0.6269 0.6666 RCIM_Cubic>
         ]}
        GreenCurve {keys [<at 0 0.1305>, {arriveTangent=0.126, interpMode=RCIM_Cubic, leaveTangent=0.126, time=0.0999, value=1.3046}, <at 0.6269 0.2094 RCIM_Cubic>]}
        RedCurve {keys [<at 0 1>,<at 0.0999 10 RCIM_Cubic>,<at 0.6269 1 RCIM_Cubic>]}
       }})
     Scale Factor=&MakeFloat4FromVectorAndFloat(
      "Vector (RGB)"=&VectorFromFloat(Value=&FloatFromCurve(
        FloatCurve=DataInterface() {propertyValues {Curve {keys [
            {interpMode=RCIM_Cubic, tangentMode=RCTM_Break, time=0, value=1}
            {arriveTangent=-2.2558, interpMode=RCIM_Cubic, tangentMode=RCTM_Break, time=1, value=0}
           ]}}}
        Scale Curve=10
       ))
      "Float (Alpha)"=&FloatFromCurve(FloatCurve=DataInterface() {propertyValues {Curve {keys [<at 0.4 1 RCIM_Cubic>,<at 1 0>]}}})
     )
    )
    Create Color Attribute=true
    Create Scale Attribute=true
    Scale=&VectorFromFloat(Value=&FloatFromCurve(
      FloatCurve=DataInterface() {propertyValues {Curve {keys [
          {arriveTangent=4.7021, interpMode=RCIM_Cubic, leaveTangent=4.7021, tangentMode=RCTM_User, time=0, value=0}
          {arriveTangent=2.0471, interpMode=RCIM_Cubic, leaveTangent=2.0471, tangentMode=RCTM_User, time=0.3, value=1}
          {interpMode=RCIM_Cubic, tangentMode=RCTM_User, time=1, value=0.2}
         ]}}}
      Scale Curve=40
     ))
    Write to Module Namespace=true
   ) {module=FN_Mesh_Attributes002}
   &FN_Sprite_Attributes(
    Color=&ColorFromCurve(ColorCurve=DataInterface() {propertyValues {
       AlphaCurve {keys [<at 0 1 RCIM_Cubic>,<at 1 0 RCIM_Cubic>]}
       BlueCurve {keys [<at 0 17.002 RCIM_Cubic>,<at 0.3345 2.4494 RCIM_Cubic>]}
       GreenCurve {keys [<at 0 5.8466 RCIM_Cubic>,<at 0.3345 1.2034 RCIM_Cubic>]}
       RedCurve {keys [<at 0 20 RCIM_Cubic>,<at 0.3345 10 RCIM_Cubic>]}
      }})
    Create Color Attribute=true
    Create SpriteSize Attribute=true
    SpriteSize=&Multiply_Vector2D_ByFloat(Float=1.5, Vector2D=$Particles.SpriteSize)
    Write to Module Namespace=true
   ) {module=FN_Sprite_Attributes002}
   /SetVariables(Particles.ShockwaveDynmat:Vector4f=&MakeVector4(
     X=&FloatFromCurve(FloatCurve=DataInterface() {propertyValues {Curve {keys [
          {interpMode=RCIM_Cubic, tangentMode=RCTM_User, time=0, value=1}
          {interpMode=RCIM_Cubic, tangentMode=RCTM_User, time=0.3172, value=0.9}
          {interpMode=RCIM_Cubic, tangentMode=RCTM_User, time=1, value=1}
         ]}}})
     Y=&FloatFromCurve(FloatCurve=DataInterface() {propertyValues {Curve {keys [<at 0 15>,<at 1 0>]}}})
     Z=0
     W=0
    )) {isSetParameters=true}
   &FN_Mesh_Attributes(
    Color=&ColorFromCurve(ColorCurve=DataInterface() {propertyValues {
       AlphaCurve {keys [<at 0 1 RCIM_Cubic>,<at 1 0 RCIM_Cubic>]}
       BlueCurve {keys [<at 0 2.7 RCIM_Cubic>,<at 0.1545 0.5364>,<at 0.3642 0.4297>,<at 1 0.0487 RCIM_Cubic>]}
       GreenCurve {keys [<at 0 0.4681 RCIM_Cubic>,<at 0.1545 0.1348>,<at 0.3642 0.2402>,<at 1 0.0167 RCIM_Cubic>]}
       RedCurve {keys [<at 0 5 RCIM_Cubic>,<at 0.1545 2>,<at 0.3642 1>,<at 1 0.1927 RCIM_Cubic>]}
      }})
    Create Color Attribute=true
    Create Scale Attribute=true
    Scale=&Multiply_VectorByFloat(
     Vector=$Particles.Scale
     Float=&FloatFromCurve(
      FloatCurve=DataInterface() {propertyValues {Curve {keys [
          {arriveTangent=3.7515, interpMode=RCIM_Cubic, leaveTangent=3.7515, tangentMode=RCTM_User, time=0, value=0}
          {interpMode=RCIM_Cubic, tangentMode=RCTM_User, time=0.25, value=1}
          <at 1 0.5 RCIM_Cubic>
         ]}}}
      Scale Curve=40
     )
    )
    Write to Module Namespace=true
   )
  ]
  rendererClasses [
   &NiagaraSpriteRendererProperties
   &NiagaraLightRendererProperties
   &NiagaraMeshRendererProperties
  ]
  renderers [
   {rendererClass=&NiagaraSpriteRendererProperties}
   {rendererClass=&NiagaraSpriteRendererProperties, rendererIndex=1}
   {rendererClass=&NiagaraSpriteRendererProperties, rendererIndex=2}
   {rendererClass=&NiagaraSpriteRendererProperties, rendererIndex=3}
   {rendererClass=&NiagaraLightRendererProperties, rendererIndex=4}
   {rendererClass=&NiagaraMeshRendererProperties, rendererIndex=5}
   {rendererClass=&NiagaraMeshRendererProperties, rendererIndex=6}
   {rendererClass=&NiagaraSpriteRendererProperties, rendererIndex=7}
   {rendererClass=&NiagaraMeshRendererProperties, rendererIndex=8}
  ]
 }
 ShellBurst {
  emitterUpdateScript [
   &EmitterState()
   &SpawnBurst_Instantaneous(Use Spawn Probability=false, UseLoopCountLimit=false)
  ]
  particleSpawnScript [
   &InitializeParticle(
    Color=<rgba 1 1 1 0.5>
    Lifetime=3.5
    Mesh Scale Mode=::NewEnumerator3
    Mesh Uniform Scale=26.2282
    Position Offset=<0 100 200>
    Sprite Size=&Multiply_Vector2D_ByFloat(Float=5, Vector2D=<300 650>)
    Sprite Size Mode=::NewEnumerator4
    UsePositionOffset=true
    Write Material Random=true
   )
   &CameraOffset(Camera Offset Amount=300)
   /SetVariables(
    Particles.OffsetPosition2:NiagaraPosition=&AddVectorToPosition(Position=/Niagara/DynamicInputs/Helpers/SimulationPosition(), Vector=<0 100 200>)
    Particles.Sprite2Alignment:Vector3f=<0 1 0>
    Particles.Sprite2Facing:Vector3f=<-1 0.25 0>
    Particles.SpriteAlignment=<0 1 0>
    Particles.SpriteFacing=<1 0.25 -0.3>
    Particles.TopAlignment:Vector3f=<0 1 0>
    Particles.TopFacing:Vector3f=<0 0.4 -1>
    Particles.TopSize:Vector2f=<1500 3650>
   ) {isSetParameters=true}
  ]
  particleUpdateScript [
   &ParticleState()
   &ScaleMeshSize(Scale Factor=&VectorFromFloat(Value=&FloatFromCurve(FloatCurve=DataInterface() {propertyValues {Curve {keys [<at 0 0.8>,<at 1 1>]}}}, Scale Curve=10)))
   &DynamicMaterialParameters(
    Index 0 Param 1=0.25
    Index 0 Param 2=-1
    Index 0 Param 3=1
    Index 0 Param 4=1
    Module.Index0Param0Write=true
    Module.Index0Param1Write=true
    Module.Index0Param2Write=true
    Module.Index0Param3Write=true
   )
   &ScaleColor(
    Scale Alpha=&FloatFromCurve(FloatCurve=DataInterface() {propertyValues {Curve {keys [<at 0.0999 1.0037 RCIM_Cubic>,<at 1 1>]}}})
    ScaleA=true
    ScaleRGB=true
   )
   &SpriteFacingAndAlignment(
    Alignment Coordinate Space=::World
    Sprite Alignment=&NormalizeVector(
     Vector To Normalize=&SubtractVectorFromPosition(Position=$Engine.Owner.Position, Vector=Linked() {linkedVariable {name=User.BeamEnd_0, type=NiagaraPosition}})
     Safe Normalize=true
     Safe Fallback Threshold=0
     Safe Fallback Vector=<0 0 0.001>
    )
    WriteAlignment=true
    WriteFacing=false
   )
   &ScaleColor(
    Scale Alpha=&FloatFromCurve(FloatCurve=DataInterface() {propertyValues {Curve {
        defaultValue=1
        keys [
         {arriveTangent=9.0725, interpMode=RCIM_Cubic, leaveTangent=9.0725, tangentMode=RCTM_User, time=0, value=0}
         {arriveTangent=-0.2354, interpMode=RCIM_Cubic, leaveTangent=-0.2354, tangentMode=RCTM_User, time=0.1408, value=0.9847}
         {arriveTangent=-4.2302, interpMode=RCIM_Cubic, leaveTangent=-4.2302, tangentMode=RCTM_User, time=1, value=0}
        ]
       }}})
    ScaleA=true
    ScaleRGB=true
   ) {module=ScaleColor001}
  ]
  rendererClasses [&NiagaraSpriteRendererProperties]
  renderers [
   {rendererClass=&NiagaraSpriteRendererProperties}
   {rendererClass=&NiagaraSpriteRendererProperties, rendererIndex=1}
   {rendererClass=&NiagaraSpriteRendererProperties, rendererIndex=2}
  ]
 }
 Smoke_Sphere {
  emitterUpdateScript [
   &EmitterState()
   &SpawnRate(
    SpawnRate=&FloatFromCurve(
     FloatCurve=DataInterface() {propertyValues {Curve {keys [<at 0 1>,<at 0.05 1 RCIM_Constant>,<at 0.125 0 RCIM_Constant>]}}}
     CurveIndex=$Emitter.NormalizedLoopAge
     Scale Curve=90
    )
    Use Spawn Probability=false
   )
   &SpawnBurst_Instantaneous(Spawn Count=30, Spawn Time=0.08, Use Spawn Probability=false) {enabled=false}
  ]
  particleSpawnScript [
   &PartitionParticles(
    Particles Per Partition=16
    Partition Count Mode:ENiagara_InfiniteLoopDuration=::NewEnumerator1
    Partition Mode:ENiagara_PartitionMode=::NewEnumerator1
   )
   &InitializeParticle(
    Color Maximum=<rgba 0.2 0.1821 0.1447 1>
    Color Minimum=<rgba 0.4 0.3751 0.3553 1>
    Color Mode=::NewEnumerator2
    Lifetime=&RandomRangeFloat(Minimum=0.4, Maximum=1)
    Position=$System.GroundPosition
    Position Mode=::NewEnumerator1
    Position Offset=&MakeVector(X=0, Y=0, Z=&Multiply_Float(A=1, B=1))
    Sprite Rotation Mode=::NewEnumerator2
    Sprite Size Mode=::NewEnumerator1
    Uniform Sprite Size Max=5500
    Uniform Sprite Size Min=3750
    UsePositionOffset=false
    Write Material Random=true
   )
   /SetVariables(Particles.InitSpriteRotRate=&UniformRangedFloat(Minimum=-3, Maximum=3)) {isSetParameters=true}
   &ShapeLocation(
    MaskSpawnGroup=false
    Non Uniform Scale=<1 1 0>
    Offset Mode=::NewEnumerator0
    Override Seed=false
    Ring Radius=1000
    Shape Primitive=::NewEnumerator4
    UseOwnerScale=true
   )
   &AddVelocity(
    Module.Use Falloff Distance=false
    Override Seed=false
    Velocity Mode=::NewEnumerator1
    Velocity Origin=$System.GroundPosition
    Velocity Speed=&RandomRangeFloat(Minimum=2500, Maximum=3500)
   )
  ]
  particleUpdateScript [
   &ParticleState()
   &SpriteRotationRate(Rotation Rate=&Multiply_Float(
     A=$Particles.InitSpriteRotRate
     B=&FloatFromCurve(
      FloatCurve=DataInterface() {propertyValues {Curve {keys [
          {arriveTangent=-4.0107, interpMode=RCIM_Cubic, leaveTangent=-4.0107, tangentMode=RCTM_User, time=-0.0046, value=1.0113}
          {arriveTangent=-0.2496, interpMode=RCIM_Cubic, leaveTangent=-0.2496, tangentMode=RCTM_User, time=0.5, value=0.16}
         ]}}}
      CurveIndex=$Emitter.Age
      Scale Curve=50
     )
    ))
   &Drag(Drag=1.65, Use Linear Drag=true, Use Rotational Drag=false)
   &GravityForce(Gravity=&RandomRangeVector(Minimum=<0 0 0>, Maximum=<1 1 3000>))
   &SolveForcesAndVelocity(Clamp Velocity=false, Limit Acceleration=false)
   &ScaleSpriteSize(Scale Factor=&Vector2DFromFloat(Value=&FloatFromCurve(FloatCurve=DataInterface() {propertyValues {Curve {keys [
          {arriveTangent=6.2202, interpMode=RCIM_Cubic, leaveTangent=6.2202, tangentMode=RCTM_User, time=0, value=0.15}
          {arriveTangent=1.0025, interpMode=RCIM_Cubic, leaveTangent=1.0025, tangentMode=RCTM_User, time=0.2, value=0.8}
          {arriveTangent=0.0143, interpMode=RCIM_Cubic, leaveTangent=0.0143, tangentMode=RCTM_User, time=1.0463, value=1}
         ]}}})))
   &DynamicMaterialParameters(
    Index 0 Param 1=&Add_Float(
     A=$Particles.RandomFloat
     B=&FloatFromCurve(
      FloatCurve=DataInterface() {propertyValues {Curve {keys [
          {arriveTangent=0.3307, interpMode=RCIM_Cubic, leaveTangent=0.3307, tangentMode=RCTM_User, time=0, value=1}
          {arriveTangent=0.0529, interpMode=RCIM_Cubic, leaveTangent=0.2021, tangentMode=RCTM_User, time=1, value=1.1}
         ]}}}
      CurveIndex=&Add_Float(A=&Multiply_Float(A=$Particles.Initial.MaterialRandom, B=0.5), B=$Particles.NormalizedAge)
     )
    )
    Index 0 Param 2=&FloatFromCurve(FloatCurve=DataInterface() {propertyValues {Curve {keys [
         {arriveTangent=0.0453, interpMode=RCIM_Cubic, tangentMode=RCTM_Break, time=0, value=-1}
         {arriveTangent=0.0604, interpMode=RCIM_Cubic, leaveTangent=0.0604, tangentMode=RCTM_User, time=1, value=-1}
        ]}}})
    Index 0 Param 3=&FloatFromCurve(
     FloatCurve=DataInterface() {propertyValues {Curve {keys [{arriveTangent=0.2347, interpMode=RCIM_Cubic, leaveTangent=0.2723, tangentMode=RCTM_Break, time=0.15, value=0}]}}}
     Scale Curve=0
    )
    Index 0 Param 4=&FloatFromCurve(FloatCurve=DataInterface() {propertyValues {Curve {keys [<at 0.5 1>,<at 1 0>]}}})
    Index 1 Param 1=&FloatFromCurve(
     FloatCurve=DataInterface() {propertyValues {Curve {keys [
         {interpMode=RCIM_Cubic, leaveTangent=4.7088, tangentMode=RCTM_Break, time=0, value=0}
         {arriveTangent=4.2257, interpMode=RCIM_Cubic, leaveTangent=4.2257, tangentMode=RCTM_User, time=0.25, value=1}
        ]}}}
     Scale Curve=&RandomRangeFloat(Minimum=1.5, Maximum=1.5)
    )
    Index 1 Param 2=&FloatFromCurve(
     FloatCurve=DataInterface() {propertyValues {Curve {keys [<at 0 0.2>, {leaveTangent=0.9066, time=1, value=2}]}}}
     Scale Curve=&RandomRangeFloat(Minimum=0.5, Maximum=0.6)
    )
    Index 1 Param 3=&FloatFromCurve(FloatCurve=DataInterface() {propertyValues {Curve {keys [
         {interpMode=RCIM_Cubic, leaveTangent=6.3917, tangentMode=RCTM_Break, time=0, value=0.2}
         {arriveTangent=1.7549, interpMode=RCIM_Cubic, leaveTangent=10.6767, tangentMode=RCTM_User, time=0.25, value=1}
        ]}}})
    Index 1 Param 4=&FloatFromCurve(
     FloatCurve=DataInterface() {propertyValues {Curve {keys [{arriveTangent=-0.136, interpMode=RCIM_Cubic, leaveTangent=-0.136, tangentMode=RCTM_User, time=0, value=1}, <at 1 0.95>]}}}
     Scale Curve=&RandomRangeFloat(Minimum=1.1, Maximum=1.4)
    )
    Index 2 Param 1=&FloatFromCurve(FloatCurve=DataInterface() {propertyValues {Curve {keys [
         {interpMode=RCIM_Cubic, leaveTangent=4.2485, tangentMode=RCTM_Break, time=0, value=0}
         {arriveTangent=1.0548, interpMode=RCIM_Cubic, leaveTangent=1.0548, tangentMode=RCTM_User, time=0.17, value=0.5715}
         <at 1 1>
        ]}}})
    Index 2 Param 2=&FloatFromCurve(
     FloatCurve=DataInterface() {propertyValues {Curve {keys [
         {arriveTangent=1.9056, interpMode=RCIM_Cubic, leaveTangent=1.9056, tangentMode=RCTM_User, time=0, value=0.2}
         {arriveTangent=0.0259, interpMode=RCIM_Cubic, leaveTangent=0.0259, tangentMode=RCTM_User, time=1, value=1.4}
        ]}}}
     Scale Curve=&RandomRangeFloat(Minimum=0.2, Maximum=0.7)
    )
    Index 2 Param 3=&FloatFromCurve(FloatCurve=DataInterface() {propertyValues {Curve {keys [
         {interpMode=RCIM_Cubic, leaveTangent=2.3064, tangentMode=RCTM_Break, time=0, value=1}
         {interpMode=RCIM_Cubic, leaveTangent=11.1233, tangentMode=RCTM_Break, time=1, value=2}
        ]}}})
    Index 2 Param 4=&FloatFromCurve(
     FloatCurve=DataInterface() {propertyValues {Curve {keys [
         {arriveTangent=-0.6867, interpMode=RCIM_Cubic, leaveTangent=-0.6867, tangentMode=RCTM_User, time=0, value=1}
         {interpMode=RCIM_Cubic, leaveTangent=-1.2956, tangentMode=RCTM_Break, time=1, value=0.75}
        ]}}}
     Scale Curve=2.5
    )
    Module.Param0WriteEnabled=true
    Module.Param0WriteEnabled1=true
    Module.Param0WriteEnabled2=true
    Module.Param1WriteEnabled=false
    Module.Param1WriteEnabled1=true
    Module.Param1WriteEnabled2=true
    Module.Param2WriteEnabled=true
    Module.Param2WriteEnabled1=true
    Module.Param2WriteEnabled2=true
    Module.Param3WriteEnabled=true
    Module.Param3WriteEnabled1=true
    Module.Param3WriteEnabled2=true
    Write Parameter Index 1=true
    Write Parameter Index 2=true
   )
   &CameraOffset(Camera Offset Amount=35)
   &ScaleColor(
    Scale Alpha=&FloatFromCurve(
     FloatCurve=DataInterface() {propertyValues {Curve {keys [
         <at 0 0 RCIM_Cubic>
         {arriveTangent=1.1429, interpMode=RCIM_Cubic, leaveTangent=1.1429, time=0.1, value=1}
         {arriveTangent=-1.1111, interpMode=RCIM_Cubic, leaveTangent=-1.1111, time=0.7, value=0.8}
         <at 1 0 RCIM_Cubic>
        ]}}}
     Scale Curve=0.5
    )
    ScaleA=true
    ScaleRGB=true
   ) {module=ScaleColor001}
  ]
  rendererClasses [&NiagaraSpriteRendererProperties]
  renderers [
   {rendererClass=&NiagaraSpriteRendererProperties}
   {rendererClass=&NiagaraSpriteRendererProperties, rendererIndex=1}
  ]
 }
 SoftAirflow {rendererClasses [], renderers []}
 WindCone {
  emitterUpdateScript [
   &EmitterState(Life Cycle Mode=::NewEnumerator1, Loop Behavior=::NewEnumerator1, Loop Duration=1, UseLoopDelay=false)
   &SpawnRate(SpawnRate=5, Use Spawn Probability=false)
  ]
  particleSpawnScript [
   &InitializeParticle(
    Color=<rgba 1.5 0.4062 1.0729 1>
    Lifetime=1.5
    Mesh Scale=<2 10 10>
    Mesh Scale Mode=::NewEnumerator4
    Position=$System.ImpactPos
    Position Mode=::NewEnumerator1
    UsePositionOffset=false
    Write Material Random=true
   )
   &InitialMeshOrientation(Mesh Orientation Mode=::NewEnumerator2, Rotation=<0 0.25 0>, Use Rotation Vector=true)
  ]
  particleUpdateScript [
   &ParticleState()
   &ScaleMeshSize(Scale Factor=&VectorFromFloat(Value=&FloatFromCurve(FloatCurve=DataInterface() {propertyValues {Curve {keys [
          <at 0 0 RCIM_Cubic>
          {arriveTangent=0.6338, interpMode=RCIM_Cubic, leaveTangent=0.6338, tangentMode=RCTM_User, time=0.2, value=0.8}
          <at 1 1 RCIM_Cubic>
         ]}}})))
   &ScaleColor(
    Scale Alpha=&FloatFromCurve(
     FloatCurve=DataInterface() {propertyValues {Curve {keys [
         <at 0 0 RCIM_Cubic>
         {arriveTangent=-1.2369, interpMode=RCIM_Cubic, leaveTangent=-1.2369, tangentMode=RCTM_User, time=0.1, value=0.8}
         <at 1 0 RCIM_Cubic>
        ]}}}
     Scale Curve=2
    )
    ScaleA=true
    ScaleRGB=true
   )
   &DynamicMaterialParameters(
    Index 0 Param 1=0.7
    Index 0 Param 2=1.5
    Module.Index0Param0Write=true
    Module.Index0Param1Write=true
    Module.Index0Param2Write=true
    Module.Index0Param3Write=true
   )
  ]
  rendererClasses [&NiagaraMeshRendererProperties]
  renderers [{rendererClass=&NiagaraMeshRendererProperties, rendererIndex=0}]
 }
}
summary {
 systemName=NS_PistolCanary_BeamEnergyUp
 emitters [
  {
   emitterName=NE_MuzzleFlash
   rendererClasses [
    &NiagaraSpriteRendererProperties
    &NiagaraLightRendererProperties
    &NiagaraMeshRendererProperties
   ]
  }
  {emitterName=NE_LaserRibbon, rendererClasses [&NiagaraRibbonRendererProperties]}
  {emitterName=GPU_Sparks, simTarget=GPUComputeSim, rendererClasses [&NiagaraSpriteRendererProperties]}
  {emitterName=WindCone, rendererClasses [&NiagaraMeshRendererProperties]}
  {emitterName=AmbientSmoke, rendererClasses [&NiagaraSpriteRendererProperties]}
  {emitterName=SoftAirflow, rendererClasses []}
  {emitterName=DirectionalBurst, rendererClasses [&NiagaraMeshRendererProperties]}
  {emitterName=Smoke_Sphere, rendererClasses [&NiagaraSpriteRendererProperties]}
  {emitterName=ShellBurst, rendererClasses [&NiagaraSpriteRendererProperties]}
  {emitterName=GlowFlare, rendererClasses [&NiagaraSpriteRendererProperties]}
  {emitterName=GlowFlare001, rendererClasses [&NiagaraSpriteRendererProperties]}
 ]
 userVariables [
  User.BeamEnd_0 {type=NiagaraPosition, description="", defaultValue {}}
  User.Duration {type=NiagaraFloat, defaultValue=3.8, description=""}
 ]
}
